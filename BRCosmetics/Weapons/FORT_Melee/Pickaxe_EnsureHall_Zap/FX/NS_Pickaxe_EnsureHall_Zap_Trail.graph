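# Niagara swing trail for the EnsureHall Zap pickaxe, built from two short-lived ribbon emitters that follow the weapon animation.
# Both use the FN anim-trail modules to spawn and update particles between bone or socket pairs on a skeletal mesh, with a camera offset.
# Each fades through its own color curve, one stronger and one fainter, then a second ScaleColor adjusts tint and alpha by time of day.
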NiagaraSystem /BRCosmetics/Weapons/FORT_Melee/Pickaxe_EnsureHall_Zap/FX/NS_Pickaxe_EnsureHall_Zap_Trail
dependencies {
 usedDataInterfaces [&NiagaraDataInterfaceColorCurve, &NiagaraDataInterfaceCurve, &NiagaraDataInterfaceSkeletalMesh]
 usedDynamicInputs [&FloatFromCurve]
 usedModules [
  &self:SystemSpawnScript.SystemScriptSource.SystemScriptGraph.NiagaraNodeAssignment_2.SetVariables_#1
  &self:SystemSpawnScript.SystemScriptSource.SystemScriptGraph.NiagaraNodeAssignment_0.SetVariables_#2
  &SystemState
 ]
}
emitters {
 Ribbon_SideVector {
  emitterUpdateScript [
   &EmitterState(Life Cycle Mode=::NewEnumerator1, Loop Behavior=::NewEnumerator1, Loop Duration=0.2, UseLoopDelay=false)
   &FN_SpawnAnimTrail()
  ]
  particleSpawnScript [
   &FN_InitializeAnimTrailBoneIndecies(First Bone or Socket Index=1, Second Bone or Socket Index=0)
   &FN_InitializeAnimTrail(Color=<rgba 1 1 1 1.25>, Lifetime=0.15, SkeletalMesh=$System.Skeletal Mesh)
   &CameraOffset(Camera Offset Amount=5)
  ]
  particleUpdateScript [
   &FN_UpdateAnimTrail()
   &ScaleColor(
    Linear Color Curve=DataInterface() {propertyValues {
      AlphaCurve {keys [<at 0 2 RCIM_Cubic>,<at 0.3447 1>,<at 0.9794 0 RCIM_Cubic>]}
      BlueCurve {keys [{arriveTangent=0.1336, interpMode=RCIM_Cubic, time=0.009, value=0.1927}, <at 0.0325 0.1823>, <at 0.1986 0.0938>, <at 0.9108 0.0208>]}
      GreenCurve {keys [{arriveTangent=-0.4141, interpMode=RCIM_Cubic, time=0.009, value=0.1927}, <at 0.0325 0.1823>, <at 0.1986 0.0938>, <at 0.9108 0.0208>]}
      RedCurve {keys [{arriveTangent=-16.267, interpMode=RCIM_Cubic, time=0.009, value=0.1927}, <at 0.0325 0.1823>, <at 0.1986 0.0938>, <at 0.9108 0.0208>]}
     }}
    Scale Mode=::NewEnumerator2
   )
   &ScaleColor(
    Scale Alpha=&FN_Select_Float_from_Time_Of_Day(Morning=0.9, Day=1, Evening=0.9, Night=0.75)
    Scale RGB=&MakeVectorFromLinearColorRGB(Color=&FN_Select_Color_from_Time_Of_Day(
      Morning=<rgba 0.8 0.7415 0.6111 1>
      Day=<rgba 1 1 1 1>
      Evening=<rgba 0.7739 0.4702 0.448 1>
      Night=<rgba 0.2451 0.5043 1 1>
      Evening Threshold=21
     ))
    ScaleA=true
    ScaleRGB=true
   ) {module=ScaleColor001}
  ]
  rendererClasses [&NiagaraRibbonRendererProperties]
  renderers [{rendererClass=&NiagaraRibbonRendererProperties}]
 }
 Ribbon_SideVector001 {
  emitterUpdateScript [
   &EmitterState(Life Cycle Mode=::NewEnumerator1, Loop Behavior=::NewEnumerator1, Loop Duration=0.2, UseLoopDelay=false)
   &FN_SpawnAnimTrail()
  ]
  particleSpawnScript [
   &FN_InitializeAnimTrailBoneIndecies()
   &FN_InitializeAnimTrail(Color=<rgba 0.927 0.8697 0.9478 0.35>, Lifetime=0.15, SkeletalMesh=$System.Skeletal Mesh)
   &CameraOffset(Camera Offset Amount=5)
  ]
  particleUpdateScript [
   &FN_UpdateAnimTrail()
   &ScaleColor(
    Linear Color Curve=DataInterface() {propertyValues {
      AlphaCurve {keys [<at 0 0.75 RCIM_Cubic>,<at 0.9794 0 RCIM_Cubic>]}
      BlueCurve {keys [<at 0.0236 0.3767>]}
      GreenCurve {keys [<at 0.0236 0.3767>]}
      RedCurve {keys [<at 0.0236 0.4087>]}
     }}
    Scale Mode=::NewEnumerator2
   )
   &ScaleColor(
    Scale Alpha=&FN_Select_Float_from_Time_Of_Day(Morning=0.75, Day=1, Evening=0.75, Night=0.7)
    Scale RGB=&MakeVectorFromLinearColorRGB(Color=&FN_Select_Color_from_Time_Of_Day(
      Morning=<rgba 0.8 0.7415 0.6111 1>
      Day=<rgba 1 1 1 1>
      Evening=<rgba 0.7739 0.4702 0.448 1>
      Night=<rgba 0.2451 0.5043 1 1>
      Evening Threshold=21
     ))
    ScaleA=true
    ScaleRGB=true
   ) {module=ScaleColor001}
  ]
  rendererClasses [&NiagaraRibbonRendererProperties]
  renderers [{rendererClass=&NiagaraRibbonRendererProperties}]
 }
}
summary {
 systemName=NS_Pickaxe_EnsureHall_Zap_Trail
 emitters [
  {emitterName=Ribbon_SideVector, rendererClasses [&NiagaraRibbonRendererProperties]}
  {emitterName=Ribbon_SideVector001, rendererClasses [&NiagaraRibbonRendererProperties]}
 ]
 userVariables [User.RibbonIDOffset {type=NiagaraInt32}]
}
