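# Animation Blueprint for the EnsureHall Zap pickaxe that spins its five base_lance bones during melee animations.
# Each update, a melee flag delayed 0.1 seconds snaps rotation speed high, otherwise easing it to zero at a front-end-dependent rate.
# Speed times delta time accumulates into a wrapped yaw, applied additively to each bone in parent bone space.
# Sphere_Rotation and a harvesting-combo montage flag are also set, but neither drives the AnimGraph.
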AnimBlueprint /BRCosmetics/Weapons/FORT_Melee/Pickaxe_EnsureHall_Zap/Meshes/EnsureHall_Zap_Axe_AnimBP
target_skeleton=&/EnsureHall_Zap_Axe_Skeleton
graphs [
 AnimGraph {nodes [
   AnimGraphNode_SubInput_0 = Input Pose()
   n1 = "Transform (Modify) Bone - Bone: base_lance_01"!(
    ComponentPose=Local To Component(LocalPose=@AnimGraphNode_SubInput_0)
    bAlphaBoolEnabled:bool=True
    Translation="0,0,0"
    Rotation=Get Spinning_Rotator() -> (Spinning_Rotator:Rotator)
    Scale="1,1,1"
   ) -> (Pose) {
    alphaInputType=Bool
    rotationMode=BMM_Additive
    rotationSpace=BCS_ParentBoneSpace
    alphaBoolBlend {blendInTime=0.1, blendOption=Linear, blendOutTime=0.8}
    boneToModify {boneName=base_lance_01}
   }
   n2 = "Transform (Modify) Bone - Bone: base_lance_02"!(ComponentPose=@n1, bAlphaBoolEnabled:bool=True, Translation="0,0,0", Rotation=Get Spinning_Rotator(), Scale="1,1,1") -> (Pose) {
    alphaInputType=Bool
    rotationMode=BMM_Additive
    rotationSpace=BCS_ParentBoneSpace
    alphaBoolBlend {blendInTime=0.1, blendOption=Linear, blendOutTime=0.8}
    boneToModify {boneName=base_lance_02}
   }
   n3 = "Transform (Modify) Bone - Bone: base_lance_03"!(ComponentPose=@n2, bAlphaBoolEnabled:bool=True, Translation="0,0,0", Rotation=Get Spinning_Rotator(), Scale="1,1,1") -> (Pose) {
    alphaInputType=Bool
    rotationMode=BMM_Additive
    rotationSpace=BCS_ParentBoneSpace
    alphaBoolBlend {blendInTime=0.1, blendOption=Linear, blendOutTime=0.8}
    boneToModify {boneName=base_lance_03}
   }
   n4 = "Transform (Modify) Bone - Bone: base_lance_04"!(ComponentPose=@n3, bAlphaBoolEnabled:bool=True, Translation="0,0,0", Rotation=Get Spinning_Rotator(), Scale="1,1,1") -> (Pose) {
    alphaInputType=Bool
    rotationMode=BMM_Additive
    rotationSpace=BCS_ParentBoneSpace
    alphaBoolBlend {blendInTime=0.1, blendOption=Linear, blendOutTime=0.8}
    boneToModify {boneName=base_lance_04}
   }
   n5 = "Transform (Modify) Bone - Bone: base_lance_05"!(ComponentPose=@n4, bAlphaBoolEnabled:bool=True, Translation="0,0,0", Rotation=Get Spinning_Rotator(), Scale="1,1,1") -> (Pose) {
    alphaInputType=Bool
    rotationMode=BMM_Additive
    rotationSpace=BCS_ParentBoneSpace
    alphaBoolBlend {blendInTime=0.1, blendOption=Linear, blendOutTime=0.8}
    boneToModify {boneName=base_lance_05}
   }
   Output Pose(Result=Component To Local(ComponentPose=@n5)) {layerGroup=None}
  ]}
 EventGraph {nodes [
   n1 = Event BlueprintUpdateAnimation()
   > Set Delta Time X_0(Delta Time X_0:float=@n1.DeltaTimeX) -> (Output_Get:float)
   > n2 = Sequence!(execute) -> (then_0, then_1, then_2:Exec, then_3:Exec, then_4:Exec, then_5:Exec)
   > n3 = Branch(Condition=Get bIsInFrontEnd())
   > n4 = Set InterpSpeed(InterpSpeed:double=10) -> (Output_Get:double)
   n5 = Branch(execute=[@n4.then, @n6.then], Condition=Get bIsPlayingMeleeAnim())
   > Delay(Duration=0.1)
   > Set animation plus delay(animation plus delay:bool=true) -> (Output_Get:bool)
   Set animation plus delay(execute=@n5.else, animation plus delay=false)
   n6 = Set InterpSpeed(execute=@n3.else, InterpSpeed=2)
   n7 = GetAnimInstance(self=GetOwnerSkeletalMeshComponent())
   Set is Animation player melee combo(
    execute=@n2.then_4
    is Animation player melee combo:bool="Equal (Object)"(
     A=GetCurrentActiveMontage!(self:Anim Instance Object Reference=@n7) -> (ReturnValue:Anim Montage Object Reference)
     B=/Game/Animation/Game/MainPlayer/Combat/Melee/OneHanded/CAS/CAS_Pickaxe_Harvesting_Combo_CMM_M
    ) {is=title K2Node_CallFunction}
   ) -> (Output_Get:bool)
   n8 = Branch(execute=@n2.then_1, Condition=Get animation plus delay() -> (animation plus delay:bool))
   > Set RotationSpeed(RotationSpeed:double=3000) -> (Output_Get:double)
   Set RotationSpeed(
    execute=@n8.else
    RotationSpeed=FInterpTo(
     Current=Get RotationSpeed() -> (RotationSpeed:double)
     Target=0
     DeltaTime=GetWorldDeltaSeconds!() -> (ReturnValue:double)
     InterpSpeed=Get InterpSpeed() -> (InterpSpeed:double)
    )
   )
   n9 = float * float(A=Get Delta Time X_0() -> (Delta Time X_0:float), B=Get RotationSpeed()) {is=title K2Node_PromotableOperator}
   n10 = Get current_velocity() -> (current_velocity:double)
   n11 = Set current_velocity(execute=@n2.then_3, current_velocity:double="% (Float)"!(A:double=float + float(A=@n9, B=@n10), B:double=360) -> (ReturnValue:double)) -> (Output_Get:double)
   > Set Spinning_Rotator(Spinning_Rotator:Rotator=MakeRotator(Yaw=@n11)) -> (Output_Get:Rotator)
   n12 = Branch(Condition=Get bIsPlayingMeleeAnim())
   > Set Sphere_Rotation(Sphere_Rotation:bool=true) -> (Output_Get:bool)
   Set Sphere_Rotation(execute=@n12.else, Sphere_Rotation=false)
   TryGetPawnOwner()
   n13 = GetGameTimeInSeconds()
   MakeRotator(Yaw=float * float(A=float * float(A=3.6, B=@n13), B=200) {is=title K2Node_CommutativeAssociativeBinaryOperator})
   Reroute Node(InputPin:Wildcard) -> (OutputPin:Wildcard)
  ]}
]
variables [
 animation plus delay {type=boolean}
 current_velocity {type=number}
 Delta Time {type=number}
 Delta Time X {type=number}
 Delta Time X_0 {type=number}
 InterpSpeed {type=number}
 is Animation player melee combo {type=boolean}
 Melee Anim Playing {type=boolean}
 RotationSpeed {type=number}
 Sphere_Rotation {type=boolean}
 Spinning_Rotator {type=Rotator, default=<rot 0 1013451.875 0>}
]
